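# Physics asset for the secondary-motion "dangly bits" of the female Fortnite DJ character: hood strings, jacket fronts and calf fur.
# Capsule bodies cover the pelvis, spine, thighs, lower-calf twist bones, arms and hands, linked by fully free constraints.
# Two six-segment hood-string chains and two box jacket flaps hang from spine_04, while fur capsules attach to the calf twists.
# These dynamic parts have locked twist and limited swing, with jacket flaps hinging on one axis.
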PhysicsAsset /BRCosmetics/Characters/Player/Female/Medium/Bodies/F_MED_Fortnite_DJ/Meshes/F_MED_Fortnite_DJ_DanglyBits_Physics
bodies {
 calf_twist_02_l [{shapeType=Capsule, length=0, radius=2}]
 calf_twist_02_r [{shapeType=Capsule, length=0, radius=2}]
 dyn_calfFur_l [{shapeType=Capsule, length=9.723, radius=6.4484, center=<8 0 0>, rotation=<rot 90 0 0>}]
 dyn_calfFur_r [{shapeType=Capsule, length=11.1078, radius=5.7828, center=<8 0 0>, rotation=<rot 90 0 0>}]
 dyn_hoodString_01_l [{shapeType=Capsule, length=3, radius=1.5, center=<2 0 0>, rotation=<rot 90 0 0>}]
 dyn_hoodString_01_r [{shapeType=Capsule, length=3, radius=1.5, center=<2 0 0>, rotation=<rot 90 0 0>}]
 dyn_hoodString_02_l [{shapeType=Capsule, length=3, radius=1.5, center=<2 0 0>, rotation=<rot 90 0 0>}]
 dyn_hoodString_02_r [{shapeType=Capsule, length=3, radius=1.5, center=<2 0 0>, rotation=<rot 90 0 0>}]
 dyn_hoodString_03_l [{shapeType=Capsule, length=3, radius=1.5, center=<2 0 0>, rotation=<rot 90 0 0>}]
 dyn_hoodString_03_r [{shapeType=Capsule, length=3, radius=1.5, center=<2 0 0>, rotation=<rot 90 0 0>}]
 dyn_hoodString_04_l [{shapeType=Capsule, length=3, radius=1.5, center=<2 0 0>, rotation=<rot 90 0 0>}]
 dyn_hoodString_04_r [{shapeType=Capsule, length=3, radius=1.5, center=<2 0 0>, rotation=<rot 90 0 0>}]
 dyn_hoodString_05_l [{shapeType=Capsule, length=4.75, radius=1.5, center=<3 0 0>, rotation=<rot 90 0 0>}]
 dyn_hoodString_05_r [{shapeType=Capsule, length=4.75, radius=1.5, center=<3 0 0>, rotation=<rot 90 0 0>}]
 dyn_hoodString_06_l [{shapeType=Capsule, length=1.75, radius=2.75, center=<2 0 0>, rotation=<rot 90 0 0>}]
 dyn_hoodString_06_r [{shapeType=Capsule, length=1.75, radius=2.75, center=<2 0 0>, rotation=<rot 90 0 0>}]
 dyn_jacketFront_l [{shapeType=Box, extentX=10, extentY=10, extentZ=1, center=<5 0 0>}]
 dyn_jacketFront_r [{shapeType=Box, extentX=10, extentY=10, extentZ=1, center=<5 0 0>}]
 hand_l [{shapeType=Capsule, length=6.9102, radius=3.9116, center=<6 0 0>, rotation=<rot 90 0 0>}]
 hand_r [{shapeType=Capsule, length=6.771, radius=3.7633, center=<-6 0 0>, rotation=<rot 90 0 0>}]
 lowerarm_l [{shapeType=Capsule, length=16, radius=3.75, center=<15.0734 -0.0001 0>, rotation=<rot 90 0 0>}]
 lowerarm_r [{shapeType=Capsule, length=15.25, radius=3.75, center=<-14.1694 -0.0001 0>, rotation=<rot 90 0 0>}]
 pelvis [{shapeType=Capsule, length=11.8638, radius=16.6933, center=<-10.6729 -1.5688 -0.0164>, rotation=<rot 0.732 -89.9999 0.5404>}]
 spine_01 [{shapeType=Capsule, length=6.5, radius=10}]
 spine_02 [{shapeType=Capsule, length=3.0922, radius=9.859, center=<0 1.0727 0>}]
 spine_03 [{shapeType=Capsule, length=6.1743, radius=10.75, center=<0 0.6118 0.0008>, rotation=<rot 0.0306 -89.9999 -0.0035>}]
 spine_04 [{shapeType=Capsule, length=5.601, radius=11.8611, center=<-0.7851 1.9338 0.026>, rotation=<rot 1.0438 -90 0.2893>}]
 thigh_l [{shapeType=Capsule, length=30.6126, radius=12.803, center=<-26.4071 0.0257 0.3574>, rotation=<rot 0 -86.5171 -86.4636>}]
 thigh_r [{shapeType=Capsule, length=30.4978, radius=12.7007, center=<26.6202 -0.2135 -0.0567>, rotation=<rot 0 -87.6831 -88.3257>}]
 upperarm_l [{shapeType=Capsule, length=21.25, radius=3.5, center=<12.5028 0.0001 0>, rotation=<rot 90 0 0>}]
 upperarm_r [{shapeType=Capsule, length=20.75, radius=3.5, center=<-12.8488 0.0001 0>, rotation=<rot 90 0 0>}]
}
constraints [
 {
  bone1Name=spine_04
  bone2Name=spine_03
  swing1LimitDegrees=45
  swing1Motion=Free
  swing2LimitDegrees=45
  swing2Motion=Free
  twistLimitDegrees=45
  twistMotion=Free
 }
 {
  bone1Name=spine_03
  bone2Name=spine_02
  swing1LimitDegrees=45
  swing1Motion=Free
  swing2LimitDegrees=45
  swing2Motion=Free
  twistLimitDegrees=45
  twistMotion=Free
 }
 {
  bone1Name=spine_02
  bone2Name=spine_01
  swing1LimitDegrees=45
  swing1Motion=Free
  swing2LimitDegrees=45
  swing2Motion=Free
  twistLimitDegrees=45
  twistMotion=Free
 }
 {
  bone1Name=spine_01
  bone2Name=pelvis
  swing1LimitDegrees=45
  swing1Motion=Free
  swing2LimitDegrees=45
  swing2Motion=Free
  twistLimitDegrees=45
  twistMotion=Free
 }
 {
  bone1Name=calf_twist_02_r
  bone2Name=thigh_r
  swing1LimitDegrees=45
  swing1Motion=Free
  swing2LimitDegrees=45
  swing2Motion=Free
  twistLimitDegrees=45
  twistMotion=Free
 }
 {
  bone1Name=calf_twist_02_l
  bone2Name=thigh_l
  swing1LimitDegrees=45
  swing1Motion=Free
  swing2LimitDegrees=45
  swing2Motion=Free
  twistLimitDegrees=45
  twistMotion=Free
 }
 {bone1Name=dyn_hoodString_01_l, bone2Name=spine_04, swing1LimitDegrees=20, swing2LimitDegrees=20, twistLimitDegrees=45, twistMotion=Locked}
 {
  bone1Name=dyn_hoodString_02_l
  bone2Name=dyn_hoodString_01_l
  swing1LimitDegrees=20
  swing2LimitDegrees=5
  twistLimitDegrees=45
  twistMotion=Locked
 }
 {
  bone1Name=dyn_hoodString_03_l
  bone2Name=dyn_hoodString_02_l
  swing1LimitDegrees=20
  swing2LimitDegrees=5
  twistLimitDegrees=45
  twistMotion=Locked
 }
 {
  bone1Name=dyn_hoodString_04_l
  bone2Name=dyn_hoodString_03_l
  swing1LimitDegrees=20
  swing2LimitDegrees=20
  twistLimitDegrees=45
  twistMotion=Locked
 }
 {
  bone1Name=dyn_hoodString_05_l
  bone2Name=dyn_hoodString_04_l
  swing1LimitDegrees=20
  swing2LimitDegrees=20
  twistLimitDegrees=45
  twistMotion=Locked
 }
 {
  bone1Name=dyn_hoodString_06_l
  bone2Name=dyn_hoodString_05_l
  swing1LimitDegrees=20
  swing2LimitDegrees=20
  twistLimitDegrees=45
  twistMotion=Locked
 }
 {bone1Name=dyn_hoodString_01_r, bone2Name=spine_04, swing1LimitDegrees=20, swing2LimitDegrees=20, twistLimitDegrees=45, twistMotion=Locked}
 {
  bone1Name=dyn_hoodString_02_r
  bone2Name=dyn_hoodString_01_r
  swing1LimitDegrees=20
  swing2LimitDegrees=20
  twistLimitDegrees=45
  twistMotion=Locked
 }
 {
  bone1Name=dyn_hoodString_03_r
  bone2Name=dyn_hoodString_02_r
  swing1LimitDegrees=20
  swing2LimitDegrees=5
  twistLimitDegrees=45
  twistMotion=Locked
 }
 {
  bone1Name=dyn_hoodString_04_r
  bone2Name=dyn_hoodString_03_r
  swing1LimitDegrees=20
  swing2LimitDegrees=20
  twistLimitDegrees=45
  twistMotion=Locked
 }
 {
  bone1Name=dyn_hoodString_05_r
  bone2Name=dyn_hoodString_04_r
  swing1LimitDegrees=20
  swing2LimitDegrees=20
  twistLimitDegrees=45
  twistMotion=Locked
 }
 {
  bone1Name=dyn_hoodString_06_r
  bone2Name=dyn_hoodString_05_r
  swing1LimitDegrees=20
  swing2LimitDegrees=20
  twistLimitDegrees=45
  twistMotion=Locked
 }
 {
  bone1Name=thigh_l
  bone2Name=pelvis
  swing1LimitDegrees=45
  swing1Motion=Free
  swing2LimitDegrees=45
  swing2Motion=Free
  twistLimitDegrees=45
  twistMotion=Free
 }
 {
  bone1Name=thigh_r
  bone2Name=pelvis
  swing1LimitDegrees=45
  swing1Motion=Free
  swing2LimitDegrees=45
  swing2Motion=Free
  twistLimitDegrees=45
  twistMotion=Free
 }
 {bone1Name=dyn_calfFur_l, bone2Name=calf_twist_02_l, swing1LimitDegrees=45, swing2LimitDegrees=45, twistLimitDegrees=45, twistMotion=Locked}
 {bone1Name=dyn_calfFur_r, bone2Name=calf_twist_02_r, swing1LimitDegrees=45, swing2LimitDegrees=45, twistLimitDegrees=45, twistMotion=Locked}
 {
  bone1Name=dyn_jacketFront_l
  bone2Name=spine_04
  swing1LimitDegrees=45
  swing1Motion=Locked
  swing2LimitDegrees=40
  twistLimitDegrees=45
  twistMotion=Locked
 }
 {
  bone1Name=dyn_jacketFront_r
  bone2Name=spine_04
  swing1LimitDegrees=45
  swing1Motion=Locked
  swing2LimitDegrees=40
  twistLimitDegrees=45
  twistMotion=Locked
 }
 {
  bone1Name=upperarm_r
  bone2Name=spine_04
  swing1LimitDegrees=45
  swing1Motion=Free
  swing2LimitDegrees=45
  swing2Motion=Free
  twistLimitDegrees=45
  twistMotion=Free
 }
 {
  bone1Name=lowerarm_r
  bone2Name=upperarm_r
  swing1LimitDegrees=45
  swing1Motion=Free
  swing2LimitDegrees=45
  swing2Motion=Free
  twistLimitDegrees=45
  twistMotion=Free
 }
 {
  bone1Name=upperarm_l
  bone2Name=spine_04
  swing1LimitDegrees=45
  swing1Motion=Free
  swing2LimitDegrees=45
  swing2Motion=Free
  twistLimitDegrees=45
  twistMotion=Free
 }
 {
  bone1Name=lowerarm_l
  bone2Name=upperarm_l
  swing1LimitDegrees=45
  swing1Motion=Free
  swing2LimitDegrees=45
  swing2Motion=Free
  twistLimitDegrees=45
  twistMotion=Free
 }
 {
  bone1Name=hand_r
  bone2Name=lowerarm_r
  swing1LimitDegrees=45
  swing1Motion=Free
  swing2LimitDegrees=45
  swing2Motion=Free
  twistLimitDegrees=45
  twistMotion=Free
 }
 {
  bone1Name=hand_l
  bone2Name=lowerarm_l
  swing1LimitDegrees=45
  swing1Motion=Free
  swing2LimitDegrees=45
  swing2Motion=Free
  twistLimitDegrees=45
  twistMotion=Free
 }
]
