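# Post-process style AnimBlueprint for the F_MED_Fortnite_DJ body, copying the parent mesh pose and adding secondary motion.
# Two RigidBody passes feed thigh-driven skirt constraints, front-skirt AnimDynamics chains, center-bone averaging constraints and a head-driven collar Flap.
# SkydivingExternalForceInput pushes the second RigidBody and AnimDynamics; the EventGraph's separate force variable is unread by the AnimGraph.
# When bRigidBodyEnabled is false, a sequence overrides the hood-string bones.
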AnimBlueprint /BRCosmetics/Characters/Player/Female/Medium/Bodies/F_MED_Fortnite_DJ/Meshes/F_MED_Fortnite_DJ_Body_Skeleton_AnimBlueprint
target_skeleton=&/F_MED_Fortnite_DJ_Body_Skeleton
graphs [
 AnimGraph {nodes [
   Component To Local()
   n1 = Layered blend per bone(
    BasePose=Use cached pose 'SavedPose13877'()
    BlendPoses_0=Sequence Player 'F_MED_Fortnite_DJ_Body_Skeleton_Sequence'() {sequence=&/F_MED_Fortnite_DJ_Body_Skeleton_Sequence}
   ) {layerSetup [{branchFilters [
       {boneName=dyn_hoodString_02_l}
       {boneName=dyn_hoodString_03_l}
       {boneName=dyn_hoodString_04_l}
       {boneName=dyn_hoodString_02_r}
       {boneName=dyn_hoodString_03_r}
       {boneName=dyn_hoodString_04_r}
      ]}]}
   Output Pose(Result=Blend Poses by bool(BlendPose_0=Use cached pose 'SavedPose13877'(), BlendPose_1=@n1, bActiveValue=Get bRigidBodyEnabled()) {blendType=Linear}) {layerGroup=None}
   n2 = RigidBody(ComponentPose=Local To Component(LocalPose=Copy Pose From Mesh() {bUseAttachedParent=true})) {
    bClampLinearTranslationLimitToRefPose=true
    bOverrideWorldGravity=true
    overridePhysicsAsset=&/F_MED_Fortnite_DJ_DanglyBits_Physics
    simulationSpace=BaseBoneSpace
    alphaBoolBlend {blendOption=Linear}
    baseBoneRef {boneName=pelvis}
    componentAppliedLinearAccClamp=<10 10 10>
    overrideWorldGravity {z=-980}
   }
   n3 = Break Fort Anim Input Skydiving External Force(FortAnimInput_SkydivingExternalForce=Get SkydivingExternalForceInput()) -> (
    bUseSkydivingVectorForce=~
    SkydivingVectorForce=~
    HeadToPelvisDirection=~
    FloatingMultiplier=~
    FloatingAdditive=~
    DivingMultiplier=~
    DivingAdditive=~
    ParachutingMultiplier=~
    ParachutingAdditive=~
   )
   n4 = RigidBody(ComponentPose=@n2, ExternalForce=@n3.SkydivingVectorForce) {
    bClampLinearTranslationLimitToRefPose=true
    bOverrideWorldGravity=true
    overridePhysicsAsset=&/F_MED_Fortnite_DJ_Physics
    simulationSpace=BaseBoneSpace
    alphaBoolBlend {blendOption=Linear}
    baseBoneRef {boneName=root}
    componentAppliedLinearAccClamp=<1000 1000 1000>
    componentLinearAccScale {z=0.5}
    componentLinearVelScale {z=0.5}
    overrideWorldGravity {z=-980}
   }
   n5 = "Constraint - dyn_skirt_fr_l_01 (R)"!(ComponentPose=@n4, Alpha=1) -> (Pose) {
    alphaBoolBlend {blendOption=Linear}
    boneToModify {boneName=dyn_skirt_fr_l_01}
    constraintSetup [{transformType=Rotation, targetBone {boneName=thigh_l}}]
    constraintWeights [1]
   }
   n6 = "Constraint - dyn_skirt_fr_r_01 (R)"!(ComponentPose=@n5, Alpha=1) -> (Pose) {
    alphaBoolBlend {blendOption=Linear}
    boneToModify {boneName=dyn_skirt_fr_r_01}
    constraintSetup [{transformType=Rotation, targetBone {boneName=thigh_r}}]
    constraintWeights [1]
   }
   n7 = "Constraint - dyn_skirt_out_l_01 (R)"!(ComponentPose=@n6, Alpha=1) -> (Pose) {
    alphaBoolBlend {blendOption=Linear}
    boneToModify {boneName=dyn_skirt_out_l_01}
    constraintSetup [{transformType=Rotation, targetBone {boneName=thigh_l}}]
    constraintWeights [1]
   }
   n8 = "Constraint - dyn_skirt_out_r_01 (R)"!(ComponentPose=@n7, Alpha=1) -> (Pose) {
    alphaBoolBlend {blendOption=Linear}
    boneToModify {boneName=dyn_skirt_out_r_01}
    constraintSetup [{transformType=Rotation, targetBone {boneName=thigh_r}}]
    constraintWeights [1]
   }
   n9 = Anim Dynamics - Bone: dyn_skirt_fr_l_01(ComponentPose=@n8, Alpha=1, ExternalForce=@n3.SkydivingVectorForce) {
    angularDampingOverride=0.98
    bGravityOverrideInSimSpace=true
    bOverrideAngularDamping=true
    bUseGravityOverride=true
    numSolverIterationsPostUpdate=2
    numSolverIterationsPreUpdate=8
    alphaBoolBlend {blendOption=Linear}
    boundBone {boneName=dyn_skirt_fr_l_01}
    gravityOverride {z=-980}
    physicsBodyDefinitions [{
      sphereCollisionRadius=0
      boundBone {boneName=dyn_skirt_fr_l_01}
      boxExtents=<20 20 10>
      constraintSetup {twistAxis=AxisZ, angularLimitsMax=<0 0 2>, angularLimitsMin=<0 -30 -2>, angularTarget=<0 0 0>}
      localJointOffset=<0 0 -3>
     }]
   }
   n10 = Anim Dynamics - Bone: dyn_skirt_fr_l_02(ComponentPose=@n9, Alpha=1, ExternalForce=@n3.SkydivingVectorForce) {
    angularDampingOverride=0.98
    bGravityOverrideInSimSpace=true
    bOverrideAngularDamping=true
    bUseGravityOverride=true
    alphaBoolBlend {blendOption=Linear}
    boundBone {boneName=dyn_skirt_fr_l_02}
    gravityOverride {z=-980}
    physicsBodyDefinitions [{
      sphereCollisionRadius=0
      boundBone {boneName=dyn_skirt_fr_l_02}
      boxExtents=<20 20 10>
      constraintSetup {twistAxis=AxisZ, angularLimitsMax=<10 1 2>, angularLimitsMin=<-10 -30 -2>, angularTarget=<0 0 0>}
      localJointOffset=<0 0 -3>
     }]
   }
   n11 = Anim Dynamics - Bone: dyn_skirt_fr_l_03(ComponentPose=@n10, Alpha=1, ExternalForce=@n3.SkydivingVectorForce) {
    angularDampingOverride=0.98
    bGravityOverrideInSimSpace=true
    bOverrideAngularDamping=true
    bUseGravityOverride=true
    alphaBoolBlend {blendOption=Linear}
    boundBone {boneName=dyn_skirt_fr_l_03}
    gravityOverride {z=-980}
    physicsBodyDefinitions [{
      sphereCollisionRadius=0
      boundBone {boneName=dyn_skirt_fr_l_03}
      boxExtents=<20 20 10>
      constraintSetup {twistAxis=AxisZ, angularLimitsMax=<10 0 2>, angularLimitsMin=<-10 -30 -2>, angularTarget=<0 0 0>}
      localJointOffset=<0 0 -3>
     }]
   }
   n12 = Anim Dynamics - Bone: dyn_skirt_fr_l_04(ComponentPose=@n11, Alpha=1, ExternalForce=@n3.SkydivingVectorForce) {
    angularDampingOverride=0.98
    bGravityOverrideInSimSpace=true
    bOverrideAngularDamping=true
    bUseGravityOverride=true
    alphaBoolBlend {blendOption=Linear}
    boundBone {boneName=dyn_skirt_fr_l_04}
    gravityOverride {z=-980}
    physicsBodyDefinitions [{
      sphereCollisionRadius=0
      boundBone {boneName=dyn_skirt_fr_l_04}
      boxExtents=<20 20 10>
      constraintSetup {twistAxis=AxisZ, angularLimitsMax=<0 0 0>, angularLimitsMin=<0 -30 0>, angularTarget=<0 0 0>}
      localJointOffset=<0 0 -3>
     }]
   }
   n13 = Anim Dynamics - Bone: dyn_skirt_fr_r_01(ComponentPose=@n12, Alpha=1, ExternalForce=@n3.SkydivingVectorForce) {
    angularDampingOverride=0.98
    bGravityOverrideInSimSpace=true
    bOverrideAngularDamping=true
    bUseGravityOverride=true
    alphaBoolBlend {blendOption=Linear}
    boundBone {boneName=dyn_skirt_fr_r_01}
    gravityOverride {z=-980}
    physicsBodyDefinitions [{
      sphereCollisionRadius=0
      boundBone {boneName=dyn_skirt_fr_r_01}
      boxExtents=<20 20 10>
      constraintSetup {twistAxis=AxisZ, angularLimitsMax=<0 0 2>, angularLimitsMin=<0 -30 -2>, angularTarget=<0 0 0>}
      localJointOffset=<0 0 3>
     }]
   }
   n14 = Anim Dynamics - Bone: dyn_skirt_fr_r_02(ComponentPose=@n13, Alpha=1, ExternalForce=@n3.SkydivingVectorForce) {
    angularDampingOverride=0.98
    bGravityOverrideInSimSpace=true
    bOverrideAngularDamping=true
    bUseGravityOverride=true
    alphaBoolBlend {blendOption=Linear}
    boundBone {boneName=dyn_skirt_fr_r_02}
    gravityOverride {z=-980}
    physicsBodyDefinitions [{
      sphereCollisionRadius=0
      boundBone {boneName=dyn_skirt_fr_r_02}
      boxExtents=<20 20 10>
      constraintSetup {twistAxis=AxisZ, angularLimitsMax=<10 1 2>, angularLimitsMin=<-10 -30 -2>, angularTarget=<0 0 0>}
      localJointOffset=<0 0 3>
     }]
   }
   n15 = Anim Dynamics - Bone: dyn_skirt_fr_r_03(ComponentPose=@n14, Alpha=1, ExternalForce=@n3.SkydivingVectorForce) {
    angularDampingOverride=0.98
    bGravityOverrideInSimSpace=true
    bOverrideAngularDamping=true
    bUseGravityOverride=true
    alphaBoolBlend {blendOption=Linear}
    boundBone {boneName=dyn_skirt_fr_r_03}
    gravityOverride {z=-980}
    physicsBodyDefinitions [{
      sphereCollisionRadius=0
      boundBone {boneName=dyn_skirt_fr_r_03}
      boxExtents=<20 20 10>
      constraintSetup {twistAxis=AxisZ, angularLimitsMax=<10 3 2>, angularLimitsMin=<-10 -30 -2>, angularTarget=<0 0 0>}
      localJointOffset=<0 0 3>
     }]
   }
   n16 = Anim Dynamics - Bone: dyn_skirt_fr_r_04(ComponentPose=@n15, Alpha=1, ExternalForce=@n3.SkydivingVectorForce) {
    angularDampingOverride=0.98
    bGravityOverrideInSimSpace=true
    bOverrideAngularDamping=true
    bUseGravityOverride=true
    alphaBoolBlend {blendOption=Linear}
    boundBone {boneName=dyn_skirt_fr_r_04}
    gravityOverride {z=-980}
    physicsBodyDefinitions [{
      sphereCollisionRadius=0
      boundBone {boneName=dyn_skirt_fr_r_04}
      boxExtents=<20 20 10>
      constraintSetup {twistAxis=AxisZ, angularLimitsMax=<0 0 0>, angularLimitsMin=<0 -30 0>, angularTarget=<0 0 0>}
      localJointOffset=<0 0 3>
     }]
   }
   n17 = "Constraint - dyn_skirt_fr_01 (PP)"!(ComponentPose=@n16, Alpha=1, ConstraintWeights_0=1, ConstraintWeights_1:float=1) -> (Pose) {
    alphaBoolBlend {blendOption=Linear}
    boneToModify {boneName=dyn_skirt_fr_01}
    constraintSetup [{targetBone {boneName=dyn_skirt_fr_l_01}}, {targetBone {boneName=dyn_skirt_fr_r_01}}]
    constraintWeights [1,1]
   }
   n18 = "Constraint - dyn_skirt_fr_02 (RR)"!(ComponentPose=@n17, Alpha=1, ConstraintWeights_0=1, ConstraintWeights_1:float=1) -> (Pose) {
    alphaBoolBlend {blendOption=Linear}
    boneToModify {boneName=dyn_skirt_fr_02}
    constraintSetup [{transformType=Rotation, targetBone {boneName=dyn_skirt_fr_l_02}}, {transformType=Rotation, targetBone {boneName=dyn_skirt_fr_r_02}}]
    constraintWeights [1,1]
   }
   n19 = "Constraint - dyn_skirt_fr_03 (RR)"!(ComponentPose=@n18, Alpha=1, ConstraintWeights_0=1, ConstraintWeights_1:float=1) -> (Pose) {
    alphaBoolBlend {blendOption=Linear}
    boneToModify {boneName=dyn_skirt_fr_03}
    constraintSetup [{transformType=Rotation, targetBone {boneName=dyn_skirt_fr_l_03}}, {transformType=Rotation, targetBone {boneName=dyn_skirt_fr_r_03}}]
    constraintWeights [1,1]
   }
   n20 = "Constraint - dyn_skirt_fr_04 (RR)"!(ComponentPose=@n19, Alpha=1, ConstraintWeights_0=1, ConstraintWeights_1:float=1) -> (Pose) {
    alphaBoolBlend {blendOption=Linear}
    boneToModify {boneName=dyn_skirt_fr_04}
    constraintSetup [{transformType=Rotation, targetBone {boneName=dyn_skirt_fr_l_04}}, {transformType=Rotation, targetBone {boneName=dyn_skirt_fr_r_04}}]
    constraintWeights [1,1]
   }
   n21 = Flap Node - Bone: dyn_collar_back!(ComponentPose:Component Space Pose Link Structure=@n20, Alpha:float=1) -> (Pose:Component Space Pose Link Structure) {
    is=title FortAnimGraphNode_Flap
    bUseClamp=true
    targetClampMin=-3.14
    targetComponent=RotationZ
    alphaBoolBlend {blendOption=Linear}
    sourceDrives [{
      bUseRange=false
      drivingCurve=null
      multiplier=-0.75
      rangeMax=0
      rangeMin=0
      remappedMax=0
      remappedMin=0
      sourceComponent=RotationZ
      useQuaternion=true
      sourceBone {boneName=head}
     }]
    targetBone {boneName=dyn_collar_back}
   }
   Save cached pose 'SavedPose13877'(Pose=Component To Local(ComponentPose=@n21)) {cacheName=SavedPose13877}
  ]}
 EventGraph {nodes [
   n1 = Branch(Condition=Get bCanPlayCustomAnimations() -> (bCanPlayCustomAnimations:bool))
   n2 = Branch(execute=[@n1.else, @n1.then], Condition=Get bIsSkyDiving())
   n3 = Branch(execute=[@n2.else, @n2.then], Condition=Get bIsSkydiveFloating())
   n4 = Get Direction Between Sockets!(
    Component:Skeletal Mesh Component Object Reference=GetOwnerSkeletalMeshComponent()
    SocketOrBoneNameFrom:Name=head
    SocketOrBoneNameTo:Name=pelvis
   ) -> (ReturnValue:Vector) {is=title K2Node_CallFunction}
   > Set skydiving vector force(skydiving vector force:Vector=vector + vector(
     A:Vector=vector * vector(A:Vector=@n4, B:Vector="75,75,75") -> (ReturnValue:Vector) {is=title K2Node_CommutativeAssociativeBinaryOperator}
     B:Vector="0, 0,2000"
    ) -> (ReturnValue:Vector) {is=title K2Node_CommutativeAssociativeBinaryOperator}) -> (Output_Get:Vector)
   n5 = Branch(execute=@n3.else, Condition=Get bIsSkydiveDiveMode())
   > Set skydiving vector force(skydiving vector force=vector + vector(A=vector * vector(A=@n4, B="180,180,200"), B="0, 0,2000"))
   Set skydiving vector force(execute=@n5.else, skydiving vector force=vector * vector(A=@n4, B="-15,-15,-15"))
   TryGetPawnOwner()
   Event BlueprintUpdateAnimation()
  ]}
]
variables [skydiving vector force {type=Vector}]
