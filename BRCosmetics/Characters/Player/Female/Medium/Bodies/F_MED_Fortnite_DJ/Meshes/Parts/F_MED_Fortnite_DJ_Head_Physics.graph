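# Physics asset for the head part of the female medium Fortnite DJ character, giving its two ears simple bodies for physics-driven motion.
# A capsule body on each dynamic ear bone is constrained to a zero-length capsule on the head bone, which acts as a sphere.
# Both constraints lock twist and the first swing axis, so each ear can rotate only on the second swing axis.
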PhysicsAsset /BRCosmetics/Characters/Player/Female/Medium/Bodies/F_MED_Fortnite_DJ/Meshes/Parts/F_MED_Fortnite_DJ_Head_Physics
bodies {
 dyn_ear_l [{shapeType=Capsule, length=6, radius=3, center=<3 0 0>, rotation=<rot 90 0 0>}]
 dyn_ear_r [{shapeType=Capsule, length=6, radius=3, center=<3 0 0>, rotation=<rot 90 0 0>}]
 head [{shapeType=Capsule, length=0, radius=4, rotation=<rot 0 -90 0>}]
}
constraints [
 {
  bone1Name=dyn_ear_r
  bone2Name=head
  swing1LimitDegrees=5
  swing1Motion=Locked
  swing2LimitDegrees=45
  twistLimitDegrees=45
  twistMotion=Locked
 }
 {
  bone1Name=dyn_ear_l
  bone2Name=head
  swing1LimitDegrees=5
  swing1Motion=Locked
  swing2LimitDegrees=45
  twistLimitDegrees=45
  twistMotion=Locked
 }
]
