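# Physics asset for the Remix face accessory of the female Fortnite DJ character, defining collision bodies and joints for right-side hair bones.
# Three chains attach to the head: three boxes for the main hair, two capsules for hairA and one capsule for the upper hair.
# Two capsules form the head body.
# Angular limits allow wide swing but little twist, with the widest swing at the main chain's root.
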PhysicsAsset /BRCosmetics/Characters/Player/Female/Medium/Bodies/F_MED_Fortnite_DJ/Meshes/Parts/F_MED_Fortnite_DJ_Remix_FaceAcc_Physics
bodies {
 dyn_hairA_01_r [{shapeType=Capsule, length=1.75, radius=1, center=<-0.7 0 0>, rotation=<rot 87.5243 -179.9999 180>}]
 dyn_hairA_02_r [{shapeType=Capsule, length=1.25, radius=0.75, center=<-0.6 0 0>, rotation=<rot 79.7977 0.0001 39.366>}]
 dyn_hair_01_r [{shapeType=Box, extentX=7.5, extentY=0.5, extentZ=6.5625, center=<-3.1428 0.2748 1.0602>, rotation=<rot -14.9999 -5.001 0>}]
 dyn_hair_02_r [{shapeType=Box, extentX=7.5, extentY=1.1875, extentZ=6.5625, center=<-1 0.0001 0.0518>, rotation=<rot 0 5.0013 5>}]
 dyn_hair_03_r [{shapeType=Box, extentX=7.5, extentY=1.1875, extentZ=6.5625, center=<-3.1267 -0.0001 -0.8374>, rotation=<rot 14.9999 0 -15.0014>}]
 dyn_hairupper_01_r [{shapeType=Capsule, length=1, radius=1, center=<0 -2 0>, rotation=<rot 0 0 -103.1269>}]
 head [
  {shapeType=Capsule, length=5.1875, radius=6, center=<2.0589 3.9694 0>, rotation=<rot 90 0 -15.0002>}
  {shapeType=Capsule, length=3.0938, radius=7.6562, center=<0 -2.8903 0>, rotation=<rot 90 0 0>}
 ]
}
constraints [
 {bone1Name=dyn_hair_02_r, bone2Name=dyn_hair_01_r, swing1LimitDegrees=45, swing2LimitDegrees=45, twistLimitDegrees=5}
 {bone1Name=dyn_hair_03_r, bone2Name=dyn_hair_02_r, swing1LimitDegrees=45, swing2LimitDegrees=45, twistLimitDegrees=5}
 {bone1Name=dyn_hair_01_r, bone2Name=head, swing1LimitDegrees=80, swing2LimitDegrees=80, twistLimitDegrees=5}
 {bone1Name=dyn_hairA_01_r, bone2Name=head, swing1LimitDegrees=45, swing2LimitDegrees=45, twistLimitDegrees=8}
 {bone1Name=dyn_hairA_02_r, bone2Name=dyn_hairA_01_r, swing1LimitDegrees=45, swing2LimitDegrees=45, twistLimitDegrees=8}
 {bone1Name=dyn_hairupper_01_r, bone2Name=head, swing1LimitDegrees=45, swing2LimitDegrees=45, twistLimitDegrees=8}
]
